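# ToyFox UI material that draws a circular reticle from a dotted circle and a thin ring.
# Reticle UV functions feed box and circle SDFs, antialiased into named-reroute masks, with 36 rotatable dots and parameterized ring size and thickness.
# Both layers are tinted pink-red and summed, with a parameter fading the ring toward white.
# Alpha-composite blending with heavily scaled-down or zeroed alpha terms gives a mostly additive look.
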Material /ToyFoxGameplay/UI/Materials/M_ToyFox_Reticle_Circles
graph {
 properties {bAutomaticallySetUsageInEditor=false, blendMode=BLEND_AlphaComposite, materialDomain=MD_UI, usedShadingModels=Default Lit}
 nodes [
  n1 = NamedRerouteUsage()
  Add(
   A=Multiply(A=NamedRerouteUsage()) {constB=0.1, desc=additive blend}
   B=Multiply(A=Multiply(A=@n1) {constB=0, desc=additive blend}, B=ScalarParameter() {defaultValue=0.2, parameterName=Dotted Circle Opacity})
  )
  n2 = Multiply(A=NamedRerouteUsage(), B=ScalarParameter() {defaultValue=0.2, parameterName=Dotted Circle Opacity})
  n3 = Constant4Vector() {constant=<rgba 1 0.0033 0.0704 1>}
  n4 = Multiply(A=NamedRerouteUsage()) {constB=0.3}
  n5 = LinearInterpolate(
   A=Constant4Vector() {constant=<rgba 1 0.0033 0.0704 1>}.RGBA
   B=Constant4Vector() {constant=<rgba 1 1 1 1>}.RGBA
   Alpha=ScalarParameter() {parameterName=Color change}
  )
  Add(A=Saturate(None=Multiply(A=@n2, B=@n3.RGBA)), B=Saturate(None=Multiply(A=@n4, B=@n5)))
  n6 = TextureCoordinate()
  n7 = MaterialFunctionCall(
   UVs=MaterialFunctionCall!(
    Reticle UVs=@n6
    Reticle Sides=Constant() {r=36}
    Reticle Scale=Constant() {r=0.76}
    Reticle Rotation=ScalarParameter() {defaultValue=0.125, parameterName=Dotted Circle Rotation}
    Prong Scale=Constant() {r=1}
    Prong Rotation=Constant()
    Spread Offset=Constant() {r=0.32}
    Positional Offset=Constant() {r=0.45}
    "Spread Input Range [Min,Max]"
    "Spread Remapped Range [Min,Max]"
   ) {materialFunction=/FortUILibrary/Systems/Reticles/MaterialFunctions/MF_UI_Reticle_UVs}.UVs
   Size=AppendVector(A=Constant() {r=0.02}, B=Constant() {r=0.06})
  ) {materialFunction=&MF_UI_Box}
  NamedRerouteDeclaration(None=Saturate(None=MaterialFunctionCall(DistanceField=@n7.SDF) {materialFunction=&MF_UI_AntiAliasing}.Inner)) {name=Dotted Circle, nodeColor=<rgba 0 0.9531 1 1>}
  n8 = MaterialFunctionCall(
   UVs=MaterialFunctionCall(
    Reticle UVs=@n6
    Reticle Sides=Constant() {r=1}
    Reticle Scale=Constant() {r=7}
    Prong Scale=Constant() {r=1}
    Spread Offset=Constant() {r=-0.5}
   ) {materialFunction=&MF_UI_Reticle_UVs}.UVs
   Size=LinearInterpolate(A=Constant() {r=0.06}, B=Constant() {r=0.107}, Alpha=ScalarParameter() {parameterName=Ring size})
   OutlineThickness=ScalarParameter() {defaultValue=0.002, parameterName=Ring thickness}
  ) {materialFunction=&MF_UI_Circle}
  NamedRerouteDeclaration(None=Saturate(None=MaterialFunctionCall(DistanceField=@n8.SDFOutline) {materialFunction=&MF_UI_AntiAliasing}.Inner)) {name=Ring, nodeColor=<rgba 0.4062 0 1 1>}
 ]
}
