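# UI material for a ToyFox reticle that draws four-way lines and arrow shapes in a red-pink tint.
# Reticle UV functions repeat an antialiased box SDF for the sticks, shifted by Spread Offset, and a masked four-sided polygon outline for the arrows, scaled by Arrow size.
# Named reroutes carry both shapes into separately weighted sums, one tinted and one labeled additive blend, under Alpha Composite blending.
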Material /ToyFoxGameplay/UI/Materials/M_ToyFox_Reticle_Lines
graph {
 properties {bAutomaticallySetUsageInEditor=false, blendMode=BLEND_AlphaComposite, materialDomain=MD_UI, usedShadingModels=Default Lit}
 nodes [
  n1 = TextureCoordinate()
  n2 = Constant() {r=4}
  n3 = Constant() {r=1}
  n4 = Constant()
  n5 = Constant() {r=1}
  n6 = Constant()
  n7 = ScalarParameter() {defaultValue=-0.05, parameterName=Spread Offset}
  n8 = Constant() {r=0.016}
  n9 = Constant() {r=0.18}
  n10 = MaterialFunctionCall(DistanceField=MaterialFunctionCall(
    UVs=MaterialFunctionCall!(
     Reticle UVs=@n1
     Reticle Sides=@n2
     Reticle Scale=@n3
     Reticle Rotation=@n4
     Prong Scale=@n5
     Prong Rotation=@n6
     Spread Offset=@n7
     Positional Offset
     "Spread Input Range [Min,Max]"
     "Spread Remapped Range [Min,Max]"
    ) {materialFunction=/FortUILibrary/Systems/Reticles/MaterialFunctions/MF_UI_Reticle_UVs}.UVs
    Size=AppendVector(A=@n8, B=@n9)
   ) {materialFunction=&MF_UI_Box}.SDF) {materialFunction=&MF_UI_AntiAliasing}
  n11 = Constant() {r=0.75}
  n12 = MaterialFunctionCall(
   Reticle UVs=@n1
   Reticle Sides=Constant() {r=4}
   Reticle Scale=LinearInterpolate(A=Constant() {r=1.1}, B=Constant() {r=0.6}, Alpha=ScalarParameter() {parameterName=Arrow size}) {constA=0.3, constB=0}
   Reticle Rotation=Constant()
   Prong Scale=Constant() {r=0.2}
   Prong Rotation=Constant()
   Spread Offset=Constant() {r=-1.1}
   Positional Offset=Constant() {r=1.7}
  ) {materialFunction=&MF_UI_Reticle_UVs}
  n13 = Multiply(A=Add(A=LinearInterpolate(Alpha=@n11) {constA=-0.5, constB=-0.7}, B=ComponentMask(None=@n12.UVs) {g=true}) {constB=-0.7}) {constB=0.01}
  n14 = MaterialFunctionCall(UVs=@n12.UVs, "Rotation Angle (0-1)"=Constant()) {materialFunction=&CustomRotator}
  n15 = Constant() {r=1}
  n16 = Constant() {r=4}
  n17 = Constant() {r=0.1}
  n18 = Constant() {r=-1}
  n19 = MaterialFunctionCall(DistanceField=Multiply(
    A=Saturate(None=@n13)
    B=MaterialFunctionCall!(UVs=@n14.Rotated Values, Size=@n15, Sides=@n16, InnerAngle, OutlineThickness=@n17, OutlineAlignment=@n18, GlowMin, GlowMax) {materialFunction=/FortUILibrary/MaterialFunctions/Shapes/MF_UI_RegularPolygon}.SDFOutline
   )) {materialFunction=&MF_UI_AntiAliasing}
  Add(
   A=Multiply(A=NamedRerouteDeclaration(None=@n10.Inner) {name=Stick Things, nodeColor=<rgba 0 0 0 1>}, B=Constant()) {constB=0.3, desc=Additive Blend}
   B=Multiply(A=NamedRerouteDeclaration(None=@n19.Inner) {name=Arrows, nodeColor=<rgba 0 0 0 1>}) {constB=0.2, desc=Addititve Blend}
  )
  n20 = NamedRerouteUsage()
  n21 = Constant() {r=0.1}
  n22 = NamedRerouteUsage()
  n23 = Constant() {r=0.6}
  Multiply(A=Constant3Vector() {constant=<rgba 1 0.0033 0.0704 1>}, B=Add(A=Multiply(A=@n20, B=@n21), B=Multiply(A=@n22, B=@n23)))
 ]
}
