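# UI material for the ToyFox reticle shape, drawing a pink-red reticle from an MSDF texture.
# The texture is sampled with normalized UI UVs, converted to a single SDF, then thresholded and anti-aliased into a shape mask.
# The mask drives a tinted color and a separate half-strength alpha, both scaled by Master Opacity.
# Alpha composite blending lets color and coverage differ, giving a partly additive look.
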Material /ToyFoxGameplay/UI/Materials/M_ToyFox_Reticle_Shape
graph {
 properties {bAutomaticallySetUsageInEditor=false, blendMode=BLEND_AlphaComposite, materialDomain=MD_UI, usedShadingModels=Default Lit}
 nodes [
  n1 = Saturate(None=NamedRerouteUsage())
  n2 = ScalarParameter() {defaultValue=0.1, parameterName=Master Opacity}
  Multiply(A=Multiply(A=@n1, B=Constant() {r=0.5}) {constB=0.7}, B=@n2)
  Multiply(A=Multiply(A=Constant4Vector() {constant=<rgba 1 0.0033 0.0704 1>}.RGBA, B=@n1), B=@n2)
  n3 = TextureSample(UVs=NamedRerouteUsage()) {samplerType=SAMPLERTYPE_LinearColor, texture=/ToyFoxGameplay/UI/Textures/T_UI_MSDF_ToxFox_Reticle}
  NamedRerouteDeclaration(None=MaterialFunctionCall(
    DistanceField=MaterialFunctionCall!(MSDF RGB=@n3.RGB) {materialFunction=/FortUILibrary/MaterialFunctions/Shapes/Modifiers/MF_UI_MSDFtoSDF}.SDF
    Threshold=Constant() {r=0.5}
   ) {materialFunction=&MF_UI_AntiAliasing}.Outer) {name=Shape, nodeColor=<rgba 0 0 0 1>}
  NamedRerouteDeclaration(None=MaterialFunctionCall() {materialFunction=&GetUserInterfaceUV}.Normalized UV) {name=UV, nodeColor=<rgba 0 0 0 1>}
 ]
}
