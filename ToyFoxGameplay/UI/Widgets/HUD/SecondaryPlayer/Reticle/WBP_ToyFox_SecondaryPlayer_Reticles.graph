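# HUD reticle widget for the ToyFox secondary player, covering basic rockets and the guided seeking missile.
# Once the event state allows, it follows the view-target pawn, showing rocket ammo and animating on fire events.
# Missile-control delegates toggle HUD elements, play a transition forward or reversed, and swap an energy arc with a timer-polled cooldown arc.
# Async ability listeners and delegates drive material parameters, with teardown on destruct.
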WidgetBlueprint /ToyFoxGameplay/UI/Widgets/HUD/SecondaryPlayer/Reticle/WBP_ToyFox_SecondaryPlayer_Reticles
animations [
 TransitionToGuidedMissile {
  binding_count=4
  length_seconds=1.0333
  object=&self:TransitionToGuidedMissile
  sequence {
   bindings [
    Image_Dot {
     id=#1
     possessed_class=Image
     tracks [
      {
       type=MovieSceneWidgetMaterialTrack
       sections [{
         type=MovieSceneParameterSection
         end_frame=31
         start_frame=0
         channels [DotSize {type=MovieSceneScriptingFloatChannel, keys [[0,0.2],[7.5,0.2],[13.5,0.5],[16.5,0]]}]
        }]
      }
      {
       type=MovieSceneColorTrack
       display_name=Color and Opacity
       sections [{
         type=MovieSceneColorSection
         end_frame=18
         start_frame=0
         channels [
          Color.R {type=MovieSceneScriptingFloatChannel, const [1,1]}
          Color.G {type=MovieSceneScriptingFloatChannel, const [1,1]}
          Color.B {type=MovieSceneScriptingFloatChannel, const [1,1]}
          Color.A {type=MovieSceneScriptingFloatChannel, keys [[0,1],[16.5,1],[18,0]]}
         ]
        }]
      }
     ]
    }
    Image_Lines {
     id=#2
     possessed_class=Image
     tracks [{
       type=MovieSceneWidgetMaterialTrack
       sections [{
         type=MovieSceneParameterSection
         end_frame=31
         start_frame=0
         channels [
          Arrow size {type=MovieSceneScriptingFloatChannel, keys [[0,0],[9,0],[28.5,1]]}
          Spread Offset {type=MovieSceneScriptingFloatChannel, keys [[0,-0.05],[25.5,-0.6]]}
         ]
        }]
      }]
    }
    Image_Circles {
     id=#3
     possessed_class=Image
     tracks [{
       type=MovieSceneWidgetMaterialTrack
       sections [{
         type=MovieSceneParameterSection
         end_frame=31
         start_frame=0
         channels [
          Ring thickness {type=MovieSceneScriptingFloatChannel, keys [[0,0.0015],[3,0.004],[7.5,0.0015],[19.5,0.0015]]}
          Dotted Circle Rotation {type=MovieSceneScriptingFloatChannel, const [0.125,1]}
          Color change {type=MovieSceneScriptingFloatChannel, keys [[0,0],[4.5,1]]}
          Ring size {type=MovieSceneScriptingFloatChannel, keys [[0,0],[12,1]]}
         ]
        }]
      }]
    }
    Image_Rockets {
     id=#4
     possessed_class=Image
     tracks [
      {
       type=MovieSceneColorTrack
       display_name=Color and Opacity
       sections [{
         type=MovieSceneColorSection
         end_frame=27
         start_frame=0
         channels [
          Color.R {type=MovieSceneScriptingFloatChannel, default=1}
          Color.G {type=MovieSceneScriptingFloatChannel, default=1}
          Color.B {type=MovieSceneScriptingFloatChannel, default=1}
          Color.A {type=MovieSceneScriptingFloatChannel, keys [[0,1],[1.5685,1],[22.6145,0]]}
         ]
        }]
      }
      {
       type=MovieSceneWidgetMaterialTrack
       sections [{
         type=MovieSceneParameterSection
         end_frame=31
         start_frame=0
         channels [
          ImageScalar {type=MovieSceneScriptingFloatChannel, const [0.8,1]}
          Master Opacity {type=MovieSceneScriptingFloatChannel, keys [[0,1],[6,0]]}
         ]
        }]
      }
     ]
    }
   ]
   display_rate {numerator=30}
  }
 }
 Rockets {
  binding_count=1
  length_seconds=5
  object=&self:Rockets
  sequence {bindings [Image_Rockets {
     id=#5
     possessed_class=Image
     tracks [{
       type=MovieSceneWidgetMaterialTrack
       sections [{
         type=MovieSceneParameterSection
         end_frame=100
         start_frame=0
         channels [RocketsRemaining {type=MovieSceneScriptingFloatChannel, keys [[0,4],[21,4],[22,3],[38,3],[39,2],[56,2],[57,1],[73,1],[74,0]]}]
        }]
      }]
    }]}
 }
 FiringRockets {
  binding_count=2
  length_seconds=0.35
  object=&self:FiringRockets
  sequence {bindings [
    Image_Lines {
     id=#6
     possessed_class=Image
     tracks [{
       type=MovieSceneWidgetMaterialTrack
       sections [{
         type=MovieSceneParameterSection
         end_frame=7
         start_frame=0
         channels [
          Arrow size {type=MovieSceneScriptingFloatChannel, keys [[0,0],[3,-0.4],[6,0]]}
          Spread Offset {type=MovieSceneScriptingFloatChannel, keys [[0,-0.05],[3,0.04],[6,-0.05]]}
         ]
        }]
      }]
    }
    Image_Circles {
     id=#7
     possessed_class=Image
     tracks [{
       type=MovieSceneWidgetMaterialTrack
       sections [{
         type=MovieSceneParameterSection
         end_frame=7
         start_frame=0
         channels [
          Ring size {type=MovieSceneScriptingFloatChannel, keys [[0,0],[3,0.3],[6,0]]}
          Ring thickness {type=MovieSceneScriptingFloatChannel, keys [[0,0.0025],[3,0.003],[6,0.0025]]}
         ]
        }]
      }]
    }
   ]}
 }
]
graphs [
 EventGraph {nodes [
   n1 = HandleBasicRocketAmmoCountChanged!() -> (OutputDelegate, then, AmmoCount:int) {is=title K2Node_CustomEvent}
   > n2 = GetDynamicMaterial(self=Get Image_Rockets() -> (Image_Rockets:Image Object Reference))
   > SetScalarParameterValue(self=@n2, ParameterName=RocketsRemaining, Value="To Float (Integer)"(InInt=@n1.AmmoCount))
   n3 = OnViewTargetChanged!() -> (
    OutputDelegate
    then
    PlayerController:Fort Player Controller Object Reference
    OldViewTarget:Actor Object Reference
    NewViewTarget:Actor Object Reference
   ) {is=title K2Node_CustomEvent}
   > n4 = Cast To FortPlayerPawn(Object=@n3.NewViewTarget)
   > Branch(Condition="Not Equal (Object)"(A:Object Reference=@n4, B:Object Reference=Get BoundPlayerPawn() -> (BoundPlayerPawn:Fort Player Pawn Object Reference)) -> (ReturnValue:bool=~) {is=title K2Node_PromotableOperator})
   > n5 = Sequence()
   UnbindCurrentPlayerPawn(execute=[@n5.then_0, @n4.CastFailed])
   n6 = Set BoundPlayerPawn(execute=@n5.then_1, BoundPlayerPawn:Fort Player Pawn Object Reference=@n4) -> (Output_Get:Fort Player Pawn Object Reference)
   > Is Valid(InputObject=@n6)
   > n7 = WaitGameplayEventToActor!(
    execute:Exec
    TargetActor:Actor Object Reference=Get BoundPlayerPawn()
    EventTag:Gameplay Tag Structure=(TagName="ToyFox.Event.Ability.Secondary.Rocket.Basic.Fired")
    OnlyTriggerOnce:bool=false
    OnlyMatchExact:bool=true
   ) -> (then:Exec, AsyncTaskProxy:Ability Async Wait Gameplay Event Object Reference, EventReceived:Exec, Payload:Gameplay Event Data Structure) {is=title K2Node_AsyncAction}
   > Set AsyncAction_WaitFireInput(AsyncAction_WaitFireInput:Ability Async Wait Gameplay Event Object Reference=@n7) -> (Output_Get:Ability Async Wait Gameplay Event Object Reference)
   > n8 = GetComponentByClass!(
    execute
    self:Actor Object Reference=Get BoundPlayerPawn()
    ComponentClass:Actor Component Class Reference=/ToyFoxGameplay/Gameplay/PawnComp/B_ToyFoxPilotPawnComponent_Secondary.B_ToyFoxPilotPawnComponent_Secondary_C
   ) -> (then, ReturnValue:B Toy Fox Pilot Pawn Component Secondary Object Reference)
   > n9 = Set BoundToyFoxPilotPawnComponent(BoundToyFoxPilotPawnComponent:B Toy Fox Pilot Pawn Component Secondary Object Reference=@n8) -> (Output_Get:B Toy Fox Pilot Pawn Component Secondary Object Reference)
   > Is Valid(InputObject=@n9)
   > Set CurrentRocketCount(CurrentRocketCount:int=GetCurrentBasicRocketAmmoCount!(self:B Toy Fox Pilot Pawn Component Secondary Object Reference=@n9) -> (Count:int=0)) -> (Output_Get:int=0)
   > Bind Event to On Basic Rocket Ammo Count Changed(
    self:B Toy Fox Pilot Pawn Component Secondary Object Reference=Get BoundToyFoxPilotPawnComponent() -> (BoundToyFoxPilotPawnComponent:B Toy Fox Pilot Pawn Component Secondary Object Reference)
    Delegate=Create Event()
   ) {is=title K2Node_AddDelegate}
   > n10 = GetDynamicMaterial(self=Get Image_Rockets())
   > SetScalarParameterValue(
    self=@n10
    ParameterName=MaxRockets
    Value="To Float (Integer)"(InInt=GetMaxBasicRocketAmmoCount!(self:B Toy Fox Pilot Pawn Component Secondary Object Reference=@n9) -> (ReturnValue:int=0))
   )
   > n11 = GetDynamicMaterial(self=Get Image_Cooldown() -> (Image_Cooldown:Image Object Reference))
   > SetScalarParameterValue(self=@n11, ParameterName=Percent, Value=1)
   PlayAnimation(execute=@n7.EventReceived, InAnimation=Get FiringRockets() -> (FiringRockets:Widget Animation Object Reference))
   n12 = OnStartedControllingMissile!() -> (OutputDelegate, then, Missile:Fort Bike Kid Object Reference) {is=title K2Node_CustomEvent}
   > Set MissileActor(MissileActor:Fort Bike Kid Object Reference=@n12.Missile) -> (Output_Get:Fort Bike Kid Object Reference)
   > n13 = GetFortHUDContextByPlayer!(
    execute
    LocalPlayer:Local Player Object Reference=GetOwningLocalPlayer!(self:Widget Object Reference) -> (ReturnValue:Local Player Object Reference)
   ) -> (then, ReturnValue:Fort HUDContext Object Reference)
   > SetHUDElementVisibility!(
    execute
    self:Fort HUDContext Object Reference=@n13
    HUDElementTags:"Gameplay Tag Container Structure (by ref)"=Get ControllingMissileHideHudTags() -> (ControllingMissileHideHudTags:Gameplay Tag Container Structure)
    bHideElements:bool=true
    ReasonName:"Name (by ref)"=ControllingMissile
   ) -> (then)
   > n14 = GetDynamicMaterial(self=Get Image_Timer() -> (Image_Timer:Image Object Reference))
   > SetScalarParameterValue(self=@n14, ParameterName=Percent, Value=1)
   > PlayAnimationForward(InAnimation=Get TransitionToGuidedMissile() -> (TransitionToGuidedMissile:Widget Animation Object Reference))
   > Clear and Invalidate Timer by Handle(Handle=Get UpdateCooldownTimerHandle() -> (UpdateCooldownTimerHandle:Timer Handle Structure))
   > n15 = WaitForAttributeChanged!(
    execute:Exec
    TargetActor:Actor Object Reference=Get MissileActor() -> (MissileActor:Fort Bike Kid Object Reference)
    Attribute:Gameplay Attribute Structure=(AttributeName="Energy",Attribute=/Script/BikeKidRuntime.FortChargingSet_BikeKidEnergy:Energy,AttributeOwner="&Class'&FortChargingSet_BikeKidEnergy'")
    OnlyTriggerOnce:bool=false
   ) -> (
    then:Exec
    AsyncTaskProxy:Ability Async Wait Attribute Changed Object Reference
    Changed:Exec
    Attribute:Gameplay Attribute Structure
    NewValue:float=~
    OldValue:float=~
   ) {is=title K2Node_AsyncAction}
   > Set AsyncAction_WaitForMissileEnergy(AsyncAction_WaitForMissileEnergy:Ability Async Wait Attribute Changed Object Reference=@n15) -> (Output_Get:Ability Async Wait Attribute Changed Object Reference)
   > SetVisibility(self=Get Image_Cooldown(), InVisibility=Hidden)
   > SetVisibility(self=Get Image_Timer(), InVisibility=SelfHitTestInvisible)
   n16 = Sequence(execute=@n15.Changed)
   > Branch(Condition="Nearly Equal (Float)"(A=Get MaxEnergy() -> (MaxEnergy:double), B=0, ErrorTolerance=0))
   > UpdateMaxEnergy!(execute, self, Actor:Actor Object Reference=Get MissileActor()) -> (then)
   n17 = GetDynamicMaterial(execute=@n16.then_1, self=Get Image_Timer())
   n18 = Get MaxEnergy()
   > SetScalarParameterValue(
    self=@n17
    ParameterName=Percent
    Value=Lerp(A=0, B=Get MaxBarPercent() -> (MaxBarPercent:double), Alpha=SafeDivide(A=@n15.NewValue, B=@n18))
   )
   OnMissileDismissed() {is=title K2Node_CustomEvent}
   > n19 = GetFortHUDContextByPlayer(LocalPlayer=GetOwningLocalPlayer())
   > SetHUDElementVisibility(self=@n19, HUDElementTags=Get ControllingMissileHideHudTags(), bHideElements=false, ReasonName=ControllingMissile)
   > EndAction(self:Ability Async Object Reference=Get AsyncAction_WaitForMissileEnergy() -> (AsyncAction_WaitForMissileEnergy:Ability Async Wait Attribute Changed Object Reference))
   > SetVisibility(self=Get Image_Cooldown(), InVisibility=SelfHitTestInvisible)
   > SetVisibility(self=Get Image_Timer(), InVisibility=Hidden)
   > PlayAnimationReverse(InAnimation=Get TransitionToGuidedMissile())
   > n20 = Set Timer by Event(
    Delegate=Create Event()
    Time=Get UpdateCooldownTickFrequency() -> (UpdateCooldownTickFrequency:double)
    bLooping=true
    bMaxOncePerFrame=true
   )
   > Set UpdateCooldownTimerHandle(UpdateCooldownTimerHandle:Timer Handle Structure=@n20) -> (Output_Get:Timer Handle Structure)
   > OnUpdateCooldown()
   n21 = HandleEventStateChanged!() -> (OutputDelegate, then, EventState:ToyFoxEventStateEnum Enum=NewEnumerator0) {is=title K2Node_CustomEvent}
   > Branch(Condition="Equal (Enum)"(A:ToyFoxEventStateEnum Enum=@n21.EventState, B:ToyFoxEventStateEnum Enum=NewEnumerator3))
   > InitializeForPlayer()
   InitializeForPlayer() {is=title K2Node_CustomEvent}
   > n22 = Cast To FortPlayerController(Object=GetOwningPlayer())
   > Bind Event to On View Target Changed(self:Fort Player Controller Object Reference=@n22, Delegate=Create Event()) {is=title K2Node_AddDelegate}
   n23 = Get Controlled Pawn!(self:Controller Object Reference=@n22) -> (ReturnValue:Pawn Object Reference) {is=title K2Node_CallFunction}
   > n24 = Is Valid(InputObject=@n23)
   n25 = OnViewTargetChanged!(
    execute=@n24.Is Not Valid
    self
    PlayerController:Fort Player Controller Object Reference
    OldViewTarget:Actor Object Reference
    NewViewTarget:Actor Object Reference=GetViewTarget!(self:Controller Object Reference=@n22) -> (ReturnValue:Actor Object Reference)
   ) -> (then) {is=title K2Node_CallFunction}
   n26 = Set BoundBikeKidControllingComponent(
    execute=[@n27.then, @n25.then]
    BoundBikeKidControllingComponent:BP Toy Fox Seek Missile Control Comp Object Reference=GetComponentByClass!(
     self=@n22
     ComponentClass=/ToyFoxGameplay/Gameplay/SecondaryPlayer/SeekingRocket/BP_ToyFox_SeekMissile_ControlComp.BP_ToyFox_SeekMissile_ControlComp_C
    ) -> (ReturnValue:BP Toy Fox Seek Missile Control Comp Object Reference)
   ) -> (Output_Get:BP Toy Fox Seek Missile Control Comp Object Reference)
   > Is Valid(InputObject=@n26)
   > Bind Event to On Started Controlling(
    self:BP Toy Fox Seek Missile Control Comp Object Reference=Get BoundBikeKidControllingComponent() -> (BoundBikeKidControllingComponent:BP Toy Fox Seek Missile Control Comp Object Reference)
    Delegate=Create Event()
   ) {is=title K2Node_AddDelegate}
   > Bind Event to On Dismissed(self:BP Toy Fox Seek Missile Control Comp Object Reference=Get BoundBikeKidControllingComponent(), Delegate=Create Event()) {is=title K2Node_AddDelegate}
   n27 = OnViewTargetChanged(execute=@n24.Is Valid, NewViewTarget=@n23)
   OnUpdateCooldown() {is=title K2Node_CustomEvent}
   > n28 = Get!(execute:Exec) -> (then:Exec, else:Exec, BoundPlayerPawn:Fort Player Pawn Object Reference)
   > n29 = GetCooldownTimeRemainingAndDuration!(
    execute
    Actor:Actor Object Reference=@n28
    CooldownTags:Gameplay Tag Container Structure=(GameplayTags=((TagName="ToyFox.Ability.Cooldown.Rocket.SeekingMissile")))
   ) -> (then, TimeRemaining:float, CooldownDuration:float)
   > n30 = GetDynamicMaterial(self=Get Image_Cooldown())
   n31 = SafeDivide(A=@n29, B=@n29.CooldownDuration)
   > SetScalarParameterValue(self=@n30, ParameterName=Percent, Value=Lerp(A=Get MaxBarPercent(), B=0, Alpha=@n31))
   > Branch(Condition="float <= float"(A=@n31, B=0.001) -> (ReturnValue=~))
   > Clear and Invalidate Timer by Handle(Handle=Get UpdateCooldownTimerHandle())
   Event Destruct()
   > UnbindCurrentPlayerPawn()
   > n32 = Sequence()
   > n33 = Cast To FortPlayerController(Object=GetOwningPlayer())
   > Unbind Event from On View Target Changed(self:Fort Player Controller Object Reference=@n33, Delegate=Create Event()) {is=title K2Node_RemoveDelegate}
   EndAction(execute=@n32.then_1, self=Get AsyncAction_WaitForMissileEnergy())
   > n34 = Get!(execute) -> (then, else, BoundBikeKidControllingComponent:BP Toy Fox Seek Missile Control Comp Object Reference)
   > Unbind Event from On Started Controlling(self:BP Toy Fox Seek Missile Control Comp Object Reference=@n34, Delegate=Create Event()) {is=title K2Node_RemoveDelegate}
   > Unbind Event from On Dismissed(self:BP Toy Fox Seek Missile Control Comp Object Reference=@n34, Delegate=Create Event()) {is=title K2Node_RemoveDelegate}
   > Set BoundBikeKidControllingComponent()
   Event Construct()
   > Parent: Construct()
   > SetVisibility(self=Get Image_Timer(), InVisibility=Hidden)
   n35 = GetComponentByClass(
    self=GetGameStateAthena!() -> (ReturnValue:Fort Game State Athena Object Reference)
    ComponentClass=/ToyFoxGameplay/Gameplay/B_ToyFoxGameStateComponent.B_ToyFoxGameStateComponent_C
   ) -> (ReturnValue:B Toy Fox Game State Component Object Reference)
   > n36 = Branch(Condition="Equal (Enum)"(
     A=GetEventState!(self:B Toy Fox Game State Component Object Reference=@n35) -> (State:ToyFoxEventStateEnum Enum=NewEnumerator0)
     B=NewEnumerator3
    ))
   > InitializeForPlayer() {is=title K2Node_CallFunction}
   Bind Event to On Event State Changed(execute=@n36.else, self:B Toy Fox Game State Component Object Reference=@n35, Delegate=Create Event()) {is=title K2Node_AddDelegate}
  ]}
 UnbindCurrentPlayerPawn {nodes [
   UnbindCurrentPlayerPawn() {is=title K2Node_FunctionEntry}
   > n1 = Sequence!(execute) -> (then_0, then_1, then_2:Exec)
   > n2 = Get!(execute) -> (then, else, AsyncAction_WaitFireInput:Ability Async Wait Gameplay Event Object Reference)
   > EndAction(self=@n2)
   n3 = Get!(execute=@n1.then_1) -> (then, else, BoundToyFoxPilotPawnComponent:B Toy Fox Pilot Pawn Component Secondary Object Reference)
   > Unbind Event from On Basic Rocket Ammo Count Changed(self:B Toy Fox Pilot Pawn Component Secondary Object Reference=@n3, Delegate=Create Event()) {is=title K2Node_RemoveDelegate}
   > Set BoundToyFoxPilotPawnComponent()
   Set BoundPlayerPawn(execute=@n1.then_2)
  ]}
 UpdateMaxEnergy {nodes [
   n1 = UpdateMaxEnergy!() -> (then, Actor:Actor Object Reference) {is=title K2Node_FunctionEntry}
   > n2 = GetAttributeSet!(
    execute
    self:Ability System Component Object Reference=GetAbilitySystemComponent!(Actor:Actor Object Reference=@n1) -> (ReturnValue:Ability System Component Object Reference)
    AttributeSetClass:Attribute Set Class Reference=&FortChargingSet_BikeKidEnergy
   ) -> (then, ReturnValue:Fort Charging Set Bike Kid Energy Object Reference)
   > Set MaxEnergy(MaxEnergy:double=Break Fort Gameplay Attribute Data!(FortGameplayAttributeData:"Fort Gameplay Attribute Data Structure (by ref)"=Get MaxEnergy!(self:Fort Charging Set Bike Kid Energy Object Reference=@n2) -> (MaxEnergy:Fort Gameplay Attribute Data Structure)) -> (
     Minimum:float=~
     Maximum:float=~
     bIsCurrentClamped:bool=~
     bIsBaseClamped:bool=~
     bShouldClampBase:bool=~
     UnclampedBaseValue:float=~
     UnclampedCurrentValue:float=~
     bShouldScaleWithMax:bool=~
     BaseValue:float=~
     CurrentValue:float=~
    ) {is=title K2Node_BreakStruct}.CurrentValue) -> (Output_Get:double)
  ]}
]
variables [
 AsyncAction_WaitFireInput {type=AbilityAsync_WaitGameplayEvent, default=null}
 AsyncAction_WaitForMissileEnergy {type=AbilityAsync_WaitAttributeChanged, default=null}
 BoundBikeKidControllingComponent {type=BP_ToyFox_SeekMissile_ControlComp_C, default=null}
 BoundPlayerPawn {type=FortPlayerPawn}
 BoundToyFoxPilotPawnComponent {type=B_ToyFoxPilotPawnComponent_Secondary_C, default=null}
 ControllingMissileHideHudTags {
  type=GameplayTagContainer
  default {gameplayTags [{tagName=ToyFox.HUD.SecondaryPlayer.QuickBar}, {tagName=ToyFox.HUD.SecondaryPlayer.Frame}]}
 }
 CurrentRocketCount {type=integer}
 MaxBarPercent {type=number, default=0.1666}
 MaxEnergy {type=number}
 MissileActor {type=FortBikeKid, default=null}
 UpdateCooldownTickFrequency {type=number, default=0.1}
 UpdateCooldownTimerHandle {type=TimerHandle}
]
widgets {
 info {inheritedWidgetCount=1, parentClass=&FortHUDElementWidget, rootWidgetClass=&Overlay, widgetCount=10}
 widgets [
  Overlay_ImageWrapper {widgetClassPath=&Overlay, widget=Overlay_ImageWrapper}
  Image_Bg {
   widgetClassPath=&Image
   bIsVariable=true
   parent=Overlay_ImageWrapper
   widget=Image_Bg
   properties {
    renderOpacity=0.2
    brush {
     bIsDynamicallyLoaded=false
     drawAs=Image
     imageType=FullColor
     mirroring=NoMirror
     resourceName=None
     resourceObject=/ToyFoxGameplay/UI/Materials/MI_ToyFox_Reticle_Bg
     tiling=NoTile
     imageSize=<500 500>
     margin=<ltrb 0 0 0 0>
     outlineSettings {
      bUseBrushTransparency=false
      roundingType=HalfHeightRadius
      width=0
      color {colorUseRule=UseColor_Specified, specifiedColor=<rgba 0 0 0 0>}
      cornerRadii=<0 0 0 0>
     }
     tintColor {colorUseRule=UseColor_Specified, specifiedColor=<rgba 1 1 1 1>}
     uVRegion {bIsValid=false, max=<0 0>, min=<0 0>}
    }
    renderTransform {angle=-30, scale=<1 1>, shear=<0 0>, translation=<0 0>}
   }
   slot {class=OverlaySlot, properties {horizontalAlignment=HAlign_Center, verticalAlignment=VAlign_Center}}
  }
  Image_Shape {
   widgetClassPath=&Image
   bIsVariable=true
   parent=Overlay_ImageWrapper
   widget=Image_Shape
   properties {brush {
     bIsDynamicallyLoaded=false
     drawAs=Image
     imageType=FullColor
     mirroring=NoMirror
     resourceName=None
     resourceObject=/ToyFoxGameplay/UI/Materials/M_ToyFox_Reticle_Shape
     tiling=NoTile
     imageSize=<400 400>
     margin=<ltrb 0 0 0 0>
     outlineSettings {
      bUseBrushTransparency=false
      roundingType=HalfHeightRadius
      width=0
      color {colorUseRule=UseColor_Specified, specifiedColor=<rgba 0 0 0 0>}
      cornerRadii=<0 0 0 0>
     }
     tintColor {colorUseRule=UseColor_Specified, specifiedColor=<rgba 1 1 1 1>}
     uVRegion {bIsValid=false, max=<0 0>, min=<0 0>}
    }}
   slot {class=OverlaySlot, properties {horizontalAlignment=HAlign_Center, verticalAlignment=VAlign_Center}}
  }
  Image_Lines {
   widgetClassPath=&Image
   bIsVariable=true
   parent=Overlay_ImageWrapper
   widget=Image_Lines
   properties {brush {
     bIsDynamicallyLoaded=false
     drawAs=Image
     imageType=FullColor
     mirroring=NoMirror
     resourceName=None
     resourceObject=/ToyFoxGameplay/UI/Materials/M_ToyFox_Reticle_Lines
     tiling=NoTile
     imageSize=<400 400>
     margin=<ltrb 0 0 0 0>
     outlineSettings {
      bUseBrushTransparency=false
      roundingType=HalfHeightRadius
      width=0
      color {colorUseRule=UseColor_Specified, specifiedColor=<rgba 0 0 0 0>}
      cornerRadii=<0 0 0 0>
     }
     tintColor {colorUseRule=UseColor_Specified, specifiedColor=<rgba 1 1 1 1>}
     uVRegion {bIsValid=false, max=<0 0>, min=<0 0>}
    }}
   slot {class=OverlaySlot, properties {horizontalAlignment=HAlign_Center, verticalAlignment=VAlign_Center}}
  }
  Image_Dot {
   widgetClassPath=&Image
   bIsVariable=true
   parent=Overlay_ImageWrapper
   widget=Image_Dot
   properties {brush {
     bIsDynamicallyLoaded=false
     drawAs=Image
     imageType=FullColor
     mirroring=NoMirror
     resourceName=None
     resourceObject=/ToyFoxGameplay/UI/Materials/MI_ToyFox_Reticle_Dot
     tiling=NoTile
     imageSize=<28 28>
     margin=<ltrb 0 0 0 0>
     outlineSettings {
      bUseBrushTransparency=false
      roundingType=HalfHeightRadius
      width=0
      color {colorUseRule=UseColor_Specified, specifiedColor=<rgba 0 0 0 0>}
      cornerRadii=<0 0 0 0>
     }
     tintColor {colorUseRule=UseColor_Specified, specifiedColor=<rgba 1 1 1 1>}
     uVRegion {bIsValid=false, max=<0 0>, min=<0 0>}
    }}
   slot {class=OverlaySlot, properties {horizontalAlignment=HAlign_Center, verticalAlignment=VAlign_Center}}
  }
  Image_Circles {
   widgetClassPath=&Image
   bIsVariable=true
   parent=Overlay_ImageWrapper
   widget=Image_Circles
   properties {brush {
     bIsDynamicallyLoaded=false
     drawAs=Image
     imageType=FullColor
     mirroring=NoMirror
     resourceName=None
     resourceObject=/ToyFoxGameplay/UI/Materials/M_ToyFox_Reticle_Circles
     tiling=NoTile
     imageSize=<550 550>
     margin=<ltrb 0 0 0 0>
     outlineSettings {
      bUseBrushTransparency=false
      roundingType=HalfHeightRadius
      width=0
      color {colorUseRule=UseColor_Specified, specifiedColor=<rgba 0 0 0 0>}
      cornerRadii=<0 0 0 0>
     }
     tintColor {colorUseRule=UseColor_Specified, specifiedColor=<rgba 1 1 1 1>}
     uVRegion {bIsValid=false, max=<0 0>, min=<0 0>}
    }}
   slot {class=OverlaySlot, properties {horizontalAlignment=HAlign_Center, verticalAlignment=VAlign_Center}}
  }
  Image_Rockets {
   widgetClassPath=&Image
   bIsVariable=true
   parent=Overlay_ImageWrapper
   widget=Image_Rockets
   properties {
    brush {
     bIsDynamicallyLoaded=false
     drawAs=Image
     imageType=FullColor
     mirroring=NoMirror
     resourceName=None
     resourceObject=/ToyFoxGameplay/UI/Materials/M_ToyFox_Reticle_Rockets
     tiling=NoTile
     imageSize=<78 42>
     margin=<ltrb 0 0 0 0>
     outlineSettings {
      bUseBrushTransparency=false
      roundingType=HalfHeightRadius
      width=0
      color {colorUseRule=UseColor_Specified, specifiedColor=<rgba 0 0 0 0>}
      cornerRadii=<0 0 0 0>
     }
     tintColor {colorUseRule=UseColor_Specified, specifiedColor=<rgba 1 1 1 1>}
     uVRegion {bIsValid=false, max=<0 0>, min=<0 0>}
    }
    renderTransform {angle=0, scale=<1.2 1.2>, shear=<0 0>, translation=<0 0>}
   }
   slot {class=OverlaySlot, properties {horizontalAlignment=HAlign_Center, verticalAlignment=VAlign_Center, padding=<ltrb 0 175 0 0>}}
  }
  Image_Timer {
   widgetClassPath=&Image
   bIsVariable=true
   parent=Overlay_ImageWrapper
   widget=Image_Timer
   properties {
    renderOpacity=0.4
    brush {
     bIsDynamicallyLoaded=false
     drawAs=Image
     imageType=FullColor
     mirroring=NoMirror
     resourceName=None
     resourceObject=/ToyFoxGameplay/UI/Materials/MI_ToyFox_Reticle_Timer
     tiling=NoTile
     imageSize=<500 500>
     margin=<ltrb 0 0 0 0>
     outlineSettings {
      bUseBrushTransparency=false
      roundingType=HalfHeightRadius
      width=0
      color {colorUseRule=UseColor_Specified, specifiedColor=<rgba 0 0 0 0>}
      cornerRadii=<0 0 0 0>
     }
     tintColor {colorUseRule=UseColor_Specified, specifiedColor=<rgba 1 1 1 1>}
     uVRegion {bIsValid=false, max=<0 0>, min=<0 0>}
    }
    renderTransform {angle=-30, scale=<1 1>, shear=<0 0>, translation=<0 0>}
   }
   slot {class=OverlaySlot, properties {horizontalAlignment=HAlign_Center, verticalAlignment=VAlign_Center}}
  }
  Image_Cooldown {
   widgetClassPath=&Image
   bIsVariable=true
   parent=Overlay_ImageWrapper
   widget=Image_Cooldown
   properties {
    renderOpacity=0.8
    brush {
     bIsDynamicallyLoaded=false
     drawAs=Image
     imageType=FullColor
     mirroring=NoMirror
     resourceName=None
     resourceObject=/ToyFoxGameplay/UI/Materials/MI_ToyFox_Reticle_Cooldown
     tiling=NoTile
     imageSize=<500 500>
     margin=<ltrb 0 0 0 0>
     outlineSettings {
      bUseBrushTransparency=false
      roundingType=HalfHeightRadius
      width=0
      color {colorUseRule=UseColor_Specified, specifiedColor=<rgba 0 0 0 0>}
      cornerRadii=<0 0 0 0>
     }
     tintColor {colorUseRule=UseColor_Specified, specifiedColor=<rgba 1 1 1 1>}
     uVRegion {bIsValid=false, max=<0 0>, min=<0 0>}
    }
    renderTransform {angle=-30, scale=<1 1>, shear=<0 0>, translation=<0 0>}
   }
   slot {class=OverlaySlot, properties {horizontalAlignment=HAlign_Center, verticalAlignment=VAlign_Center}}
  }
  TutorialHighlight {widgetClassPath=&TutorialHighlight}
 ]
}
